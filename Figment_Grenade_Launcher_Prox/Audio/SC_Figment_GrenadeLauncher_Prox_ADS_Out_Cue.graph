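# Sound cue for the aim-down-sights exit foley on the Figment proximity grenade launcher.
# It plays a single RPG ADS-out wave through a Modulator node for variation.
# It uses the weapon-foley sound class and the weapon reload attenuation settings.
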SoundCue /Figment_Grenade_Launcher_Prox/Audio/SC_Figment_GrenadeLauncher_Prox_ADS_Out_Cue
root=Modulator_0
nodes [
 {id=Modulator_0, children [WavePlayer_1]}
 {id=WavePlayer_1, properties {soundWaveAssetPtr=/Game/Sounds/Fort_Weapons/ADS/Weapon_RPG_ADS_OUT_01}}
]
properties {
 attenuationSettings=&Fort_Weapon_Reload_Att
 maxDistance=2250
 soundClassObject=&Fort_WeaponFoley
}
