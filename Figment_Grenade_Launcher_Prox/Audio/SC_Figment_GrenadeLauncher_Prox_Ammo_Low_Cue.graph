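# Sound cue for the low-ammo warning of the Figment proximity grenade launcher.
# A single wave player plays the shared high rifle low-ammo sample, routed to the degraded ranged-weapon low-ammo sound class.
# A concurrency override prevents new instances from starting once the limit is reached.
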SoundCue /Figment_Grenade_Launcher_Prox/Audio/SC_Figment_GrenadeLauncher_Prox_Ammo_Low_Cue
root=WavePlayer_66
nodes [{id=WavePlayer_66, properties {soundWaveAssetPtr=/Game/Sounds/Fort_Weapons/LowAmmo/weapon_low_ammo_hi_rifle_02}}]
properties {
 bOverrideConcurrency=true
 maxDistance=2097152
 soundClassObject=/Game/Sounds/FortSoundClassesAndModes/Fort_WeaponRanged_LowAmmo_Degraded_PSM
 concurrencyOverrides {resolutionRule=PreventNew}
}
