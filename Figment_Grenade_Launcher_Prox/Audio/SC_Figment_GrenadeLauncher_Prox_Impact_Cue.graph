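# Sound cue for the impact sound of the Figment proximity grenade launcher.
# Each variant mixes a randomly chosen, briefly delayed launcher impact with an undelayed random bullet whip-in, both pitch-modulated.
# A quality-level node uses the full wave pools on its first input and a shared, smaller-pool mix on the other three.
# It references a long bullet-impact attenuation, the enemy-damaged sound class and rifle bullet impact concurrency.
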SoundCue /Figment_Grenade_Launcher_Prox/Audio/SC_Figment_GrenadeLauncher_Prox_Impact_Cue
root=QualityLevel_0
nodes [
 {id=Delay_1, children [Modulator_2], properties {delayMax=0.1, delayMin=0.05}}
 {id=Delay_2, children [Modulator_4], properties {delayMax=0.1, delayMin=0.05}}
 {id=Mixer_1, children [Delay_1,Modulator_3], properties {inputVolume [1,1]}}
 {id=Mixer_2, children [Delay_2,Modulator_5], properties {inputVolume [1,1]}}
 {id=Modulator_2, children [Random_0], properties {pitchMax=1.2, pitchMin=1}}
 {id=Modulator_3, children [Random_2], properties {pitchMax=1.3, pitchMin=0.9}}
 {id=Modulator_4, children [Random_1], properties {pitchMax=1.2, pitchMin=1}}
 {id=Modulator_5, children [Random_4], properties {pitchMax=1.3, pitchMin=0.9}}
 {id=QualityLevel_0, children [Mixer_1,Mixer_2,Mixer_2,Mixer_2]}
 {id=Random_0, children [WavePlayer_0,WavePlayer_1,WavePlayer_2,WavePlayer_3,WavePlayer_4], properties {weights [1,1,1,1,1]}}
 {id=Random_1, children [WavePlayer_5,WavePlayer_6,WavePlayer_7], properties {weights [1,1,1]}}
 {id=Random_2, children [WavePlayer_11, WavePlayer_12, WavePlayer_13, WavePlayer_14], properties {weights [1,1,1,1]}}
 {id=Random_4, children [WavePlayer_23, WavePlayer_24, WavePlayer_25], properties {weights [1,1,1]}}
 {id=WavePlayer_0, properties {soundWaveAssetPtr=/Game/Sounds/Fort_Weapons/Launchers/Prox_GrenadeLauncher/Proximity_GrenadeLauncher_Impact_01}}
 {id=WavePlayer_1, properties {soundWaveAssetPtr=/Game/Sounds/Fort_Weapons/Launchers/Prox_GrenadeLauncher/Proximity_GrenadeLauncher_Impact_02}}
 {id=WavePlayer_11, properties {soundWaveAssetPtr=/Game/Sounds/Fort_Weapons/Impacts/bullet_whip_in_01}}
 {id=WavePlayer_12, properties {soundWaveAssetPtr=/Game/Sounds/Fort_Weapons/Impacts/bullet_whip_in_02}}
 {id=WavePlayer_13, properties {soundWaveAssetPtr=/Game/Sounds/Fort_Weapons/Impacts/bullet_whip_in_03}}
 {id=WavePlayer_14, properties {soundWaveAssetPtr=/Game/Sounds/Fort_Weapons/Impacts/bullet_whip_in_04}}
 {id=WavePlayer_2, properties {soundWaveAssetPtr=/Game/Sounds/Fort_Weapons/Launchers/Prox_GrenadeLauncher/Proximity_GrenadeLauncher_Impact_03}}
 {id=WavePlayer_23, properties {soundWaveAssetPtr=&bullet_whip_in_02}}
 {id=WavePlayer_24, properties {soundWaveAssetPtr=&bullet_whip_in_03}}
 {id=WavePlayer_25, properties {soundWaveAssetPtr=&bullet_whip_in_04}}
 {id=WavePlayer_3, properties {soundWaveAssetPtr=/Game/Sounds/Fort_Weapons/Launchers/Prox_GrenadeLauncher/Proximity_GrenadeLauncher_Impact_04}}
 {id=WavePlayer_4, properties {soundWaveAssetPtr=/Game/Sounds/Fort_Weapons/Launchers/Prox_GrenadeLauncher/Proximity_GrenadeLauncher_Impact_05}}
 {id=WavePlayer_5, properties {soundWaveAssetPtr=&Proximity_GrenadeLauncher_Impact_01}}
 {id=WavePlayer_6, properties {soundWaveAssetPtr=&Proximity_GrenadeLauncher_Impact_02}}
 {id=WavePlayer_7, properties {soundWaveAssetPtr=&Proximity_GrenadeLauncher_Impact_03}}
]
properties {
 attenuationSettings=/Game/Sounds/Attenuation/Fort_Bullet_ImpactOpponent_Lng_Att
 maxDistance=5500
 soundClassObject=/Game/Sounds/FortSoundClassesAndModes/Fort_Enemy_Damaged_PSM
 volumeMultiplier=1.5
 attenuationOverrides {falloffDistance=5000, attenuationShapeExtents {x=1000}}
 concurrencyOverrides {bLimitToOwner=true, maxCount=2, resolutionRule=StopOldest, volumeScale=0.8}
 concurrencySet [&RifleBulletImpact_Conc]
}
