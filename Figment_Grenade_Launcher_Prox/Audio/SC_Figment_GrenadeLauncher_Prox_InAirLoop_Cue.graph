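# Sound cue for the in-air flight loop of a proximity grenade launcher projectile.
# A single looping wave player plays the launcher's in-air loop wave at reduced volume, with no mixing or modulation.
# It uses projectile in-air attenuation and a third-person ranged weapon fire sound class.
# A concurrency override caps how many copies play at once.
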SoundCue /Figment_Grenade_Launcher_Prox/Audio/SC_Figment_GrenadeLauncher_Prox_InAirLoop_Cue
root=WavePlayer_0
nodes [{
  id=WavePlayer_0
  properties {bLooping=true, soundWaveAssetPtr=/Game/Sounds/Fort_Weapons/Launchers/Prox_GrenadeLauncher/Proximity_GrenadeLauncher_InAir_Loop}
 }]
properties {
 attenuationSettings=/Game/Sounds/Attenuation/Fort_Weapon_Projectile_InAir_Loop_Att
 bOverrideConcurrency=true
 maxDistance=4500
 soundClassObject=&Fort_WeaponRanged_Fire_3P_PSM
 volumeMultiplier=0.5
 concurrencyOverrides {maxCount=3}
}
