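# Sound cue for the end of the reload on the Figment proximity grenade launcher.
# A quality switch sends the first level through a modulator into an equally weighted random pick of four waves.
# The other three levels share a modulator playing only the first wave, reducing variation.
# It uses weapon foley class and reload attenuation, with a concurrency override allowing one instance, stopping the farthest on conflict.
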SoundCue /Figment_Grenade_Launcher_Prox/Audio/SC_Figment_GrenadeLauncher_Prox_Reload_End_Cue
root=QualityLevel_0
nodes [
 {id=Modulator_0, children [WavePlayer_0]}
 {id=Modulator_1, children [Random_0]}
 {id=QualityLevel_0, children [Modulator_1,Modulator_0,Modulator_0,Modulator_0]}
 {id=Random_0, children [WavePlayer_1,WavePlayer_2,WavePlayer_3,WavePlayer_4], properties {weights [1,1,1,1]}}
 {
  id=WavePlayer_0
  properties {soundWaveAssetPtr=/Game/Sounds/Fort_Weapons/Launchers/Prox_GrenadeLauncher/Proximity_GrenadeLauncher_Reload_End_01}
 }
 {
  id=WavePlayer_1
  properties {soundWaveAssetPtr=&Proximity_GrenadeLauncher_Reload_End_01}
 }
 {
  id=WavePlayer_2
  properties {soundWaveAssetPtr=/Game/Sounds/Fort_Weapons/Launchers/Prox_GrenadeLauncher/Proximity_GrenadeLauncher_Reload_End_02}
 }
 {
  id=WavePlayer_3
  properties {soundWaveAssetPtr=/Game/Sounds/Fort_Weapons/Launchers/Prox_GrenadeLauncher/Proximity_GrenadeLauncher_Reload_End_03}
 }
 {
  id=WavePlayer_4
  properties {soundWaveAssetPtr=/Game/Sounds/Fort_Weapons/Launchers/Prox_GrenadeLauncher/Proximity_GrenadeLauncher_Reload_End_04}
 }
]
properties {
 attenuationSettings=&Fort_Weapon_Reload_Att
 bOverrideConcurrency=true
 maxDistance=2250
 soundClassObject=&Fort_WeaponFoley
 volumeMultiplier=0.351
 attenuationOverrides {falloffDistance=5000, attenuationShapeExtents {x=1000}}
 concurrencyOverrides {maxCount=1, resolutionRule=StopFarthestThenPreventNew}
}
